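# Subsurface material for ice-growth rift spawner meshes, layering translucent-looking glowing ice with snow.
# The ice fakes depth using a sphere mask, counter-rotating caustic noise, atmospheric backlight and an offset fresnel core pulsed by panning noise.
# CubeGrowth functions drive spawn-timed emissive, normal and world-position-offset effects.
# Snow blends in by world-up normal, vertex-red height blend and mask green, with quality-switched specular flakes, before the game-wide shared function.
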Material /Game/Athena/Deimos/Spawners/RiftSpawners/Assets/Ice_Growth/Materials/M_IceSpawner
graph {
 properties {
  bTangentSpaceNormal=false
  bUseMaterialAttributes=true
  bUsedWithInstancedStaticMeshes=true
  pixelDepthOffsetMode=PDOM_Legacy
  shadingModel=MSM_Subsurface
  usedShadingModels=Subsurface
 }
 nodes [
  n1 = WorldPosition()
  n2 = Multiply(A=Time(), B=ScalarParameter() {defaultValue=-0.2, parameterName=NoisePannerSpeed})
  Add(
   A=Divide(
    A=MaterialFunctionCall(Float3=@n1.XYZ) {materialFunction=&BreakOutFloat3Components}.B
    B=ScalarParameter() {defaultValue=9625.9121, parameterName=NoisePannerScale}
   )
   B=@n2
  )
  Constant2Vector()
  n3 = VectorParameter() {group=Colors, parameterName=CustomColor, defaultValue=<rgba 0.4 0.0895 0.0376 1>}
  n4 = MaterialFunctionCall() {materialFunction=/Game/Athena/WinterArt/Materials/MF_SharedWinter}
  n5 = Fresnel(Normal=VertexNormalWS()) {exponent=1}
  n6 = Constant() {r=5}
  n7 = StaticSwitchParameter(
   True=VectorParameter() {group=Colors, parameterName=CustomColor, defaultValue=<rgba 0.4 0.0895 0.0376 1>}.RGB
   False=MaterialFunctionCall() {materialFunction=&MF_SharedWinter}.IceColor
  ) {group=Colors, parameterName=CustomColor}
  n8 = Multiply(B=@n7) {constA=0.5}
  n9 = Constant() {r=-0.7}
  n10 = TextureObjectParameter() {group=Tile, parameterName=R_Tile, samplerType=SAMPLERTYPE_Masks, texture=/Game/Global/Textures/Default/Tileable_Patterns/T_IceScratch1_M}
  n11 = ScalarParameter() {defaultValue=2, group=Tile, parameterName=TileTextureScale}
  n12 = Divide(A=Constant() {r=1024}, B=@n11)
  n13 = Multiply(A=TextureCoordinate(), B=@n11)
  n14 = StaticSwitchParameter(
   True=ComponentMask(None=MaterialFunctionCall(TextureObject=@n10, TextureSize=@n12) {materialFunction=&WorldAlignedTexture}.XYZ Texture) {r=true}
   False=TextureSample(UVs=@n13, Tex=@n10) {samplerType=SAMPLERTYPE_Masks, texture=&T_IceScratch1_M}.R
  ) {group=Tile, parameterName=Tile WS}
  n15 = Saturate(None=LinearInterpolate(
    A=ScalarParameter() {defaultValue=-1, group=Tile, parameterName=TileToEmissiveMin}
    B=ScalarParameter() {defaultValue=1, group=Tile, parameterName=TileToEmissiveMax}
    Alpha=@n14
   ))
  n16 = TextureSampleParameter2D() {group=Textures, parameterName=M, texture=/Game/Environments/Props/Rocks/Textures/T_Corrupt_Rock_AB_D}
  n17 = MaterialFunctionCall(
   Transition Phase=ScalarParameter() {defaultValue=0.4, group=Textures, parameterName=HeightMapTransition}
   Height Texture=@n16.R
   Contrast=ScalarParameter() {group=Textures, parameterName=HeightMapContrast}
  ) {materialFunction=&HeightLerp}
  n18 = MaterialFunctionCall(Alpha=@n17.Alpha, Min=Constant() {r=0.0479}, Max=Constant() {r=0.5}) {materialFunction=&SmoothStep}
  n19 = VertexNormalWS()
  n20 = Saturate(None=Multiply(A=ComponentMask(None=@n19) {b=true}) {constB=-1})
  n21 = SphereMask(
   A=Subtract(A=WorldPosition().XYZ, B=ActorPositionWS())
   B=VectorParameter() {group=Depth, parameterName=SphereLocation, defaultValue=<rgba 0 0 0 1>}.RGB
   Radius=Multiply(
    A=ObjectRadius()
    B=ScalarParameter() {defaultValue=1, desc=works automatically when scaling mesh, group=Depth, parameterName="SphereRadius "}
   )
   Hardness=ScalarParameter() {defaultValue=-3.5, group=Depth, parameterName=SphereHardness}
  )
  n22 = Clamp(None=LinearInterpolate(
    A=ScalarParameter() {defaultValue=3, group=Depth, parameterName=ShpereMax}
    B=ScalarParameter() {defaultValue=-9, group=Depth, parameterName=SphereMin}
    Alpha=@n21
   )) {maxDefault=5}
  n23 = Constant() {r=0.4}
  n24 = Constant() {r=-1.5456}
  n25 = AtmosphericLightVector()
  n26 = Clamp(None=DotProduct(A=Multiply(A=@n25) {constB=-1}, B=CameraVectorWS()))
  n27 = ScalarParameter() {defaultValue=8, parameterName=AtmosphereExponent}
  n28 = Multiply(
   A=Clamp(None=LinearInterpolate(A=@n23, B=@n24, Alpha=@n21)) {maxDefault=5}
   B=Multiply(
    A=Power(Base=@n26, Exp=@n27)
    B=VectorParameter() {group=Colors, parameterName="Atmospheric Light Tint ", defaultValue=<rgba 5.9913 8.3586 10 0>}.RGB
   )
  )
  n29 = Abs(None=ComponentMask(None=VertexNormalWS()) {b=true})
  n30 = WorldPosition()
  n31 = ActorPositionWS()
  n32 = Divide(
   A=ComponentMask(None=Subtract(A=@n30.XYZ, B=@n31)) {g=true, r=true}
   B=ScalarParameter() {defaultValue=7050.1357, group=Caustics, parameterName=ScaleNoiseInside}
  )
  n33 = Add(
   A=Multiply(A=ScalarParameter() {defaultValue=0.125, group=Caustics, parameterName=CausticIdleSpeed}, B=Time()) {constA=0.125}
   B=ComponentMask(None=CameraVectorWS()) {r=true}
  )
  n34 = TextureSampleParameter2D(UVs=Rotator(Coordinate=@n32, Time=@n33)) {group=Caustics, parameterName=Caustics, samplerType=SAMPLERTYPE_LinearColor, texture=/Game/Global/Textures/Default/Noise/T_Noise01}
  n35 = TextureSampleParameter2D(UVs=Rotator(Coordinate=@n32, Time=@n33) {speed=-0.25}) {group=Caustics, parameterName=Caustics, samplerType=SAMPLERTYPE_LinearColor, texture=&T_Noise01}
  n36 = Add(A=@n34.R, B=@n35.R)
  n37 = OneMinus(None=Fresnel(Normal=VertexNormalWS()) {exponent=1})
  n38 = Constant() {r=2}
  n39 = LinearInterpolate(
   A=ScalarParameter() {group=Caustics, parameterName=MinCaustic}
   Alpha=LinearInterpolate(
    B=LinearInterpolate(A=@n36, Alpha=@n29)
    Alpha=MaterialFunctionCall(Alpha=@n37, Max=@n38) {materialFunction=&SmoothStep}.Result
   ) {constA=1}
  )
  n40 = OneMinus(None=Fresnel(Normal=VertexNormalWS()) {exponent=1})
  n41 = Constant() {r=2}
  n42 = Multiply(
   A=Multiply(A=Add(A=@n22, B=@n28), B=Saturate(None=@n39))
   B=LinearInterpolate(
    A=ScalarParameter() {defaultValue=0.2, group=Depth, parameterName=FresnelMin}
    B=ScalarParameter() {defaultValue=1, group=Depth, parameterName=FresnelMax}
    Alpha=MaterialFunctionCall(Alpha=@n40, Max=@n41) {materialFunction=&SmoothStep}.Result
   )
  )
  n43 = OneMinus(None=@n17.Alpha)
  n44 = Multiply(
   A=LinearInterpolate(A=Desaturation(None=@n8, Fraction=@n9), B=@n7, Alpha=LinearInterpolate(A=@n15, Alpha=@n18.Result))
   B=Multiply(A=LinearInterpolate(Alpha=@n20) {constA=1, constB=0.2}, B=Multiply(A=@n43, B=@n42))
  )
  n45 = ScalarParameter() {defaultValue=215.0853, parameterName=DistortionICE}
  n46 = Multiply(
   A=Constant3Vector() {constant=<rgba 0.1 0.1 1 0>}
   B=LinearInterpolate(A=@n45, B=Multiply(A=@n45) {constB=-1}, Alpha=ComponentMask(None=@n14) {r=true})
  )
  n47 = Multiply(A=ScalarParameter() {defaultValue=60.0068, parameterName=DistortionICE2}, B=VertexNormalWS())
  n48 = AppendVector(
   A=AppendVector(A=ScalarParameter() {parameterName=CenterOffsetX}, B=ScalarParameter() {defaultValue=35.5543, parameterName=CenterOffsetY})
   B=ScalarParameter() {defaultValue=139.5131, parameterName=CenterOffsetZ}
  )
  n49 = MaterialFunctionCall() {materialFunction=/Engine/Functions/Engine_MaterialFunctions02/WorldPositionOffset/ObjectScale}
  n50 = Subtract(A=Add(A=WorldPosition().XYZ, B=Add(A=@n46, B=@n47)), B=Add(A=ActorPositionWS(), B=Multiply(A=@n48, B=@n49.Scale Z)))
  n51 = OneMinus(None=Fresnel(ExponentIn=ScalarParameter() {defaultValue=0.049, parameterName=IceFalloffPowerEmissiveCore}, Normal=Normalize(VectorInput=@n50)) {baseReflectFraction=0})
  n52 = VectorParameter() {parameterName=EmissiveCore, defaultValue=<rgba 0 0.5 5 1>}
  n53 = VectorParameter() {parameterName=EmissiveCore2, defaultValue=<rgba 0.1628 5.9011 10 1>}
  n54 = Multiply(
   A=Power(Base=@n51, Exp=ScalarParameter() {defaultValue=5.1604, parameterName=IceFalloffPowerEmissiveCore2})
   B=ScalarParameter() {defaultValue=8466.2832, parameterName=CoreEmissive}
  )
  n55 = Multiply(
   A=Add(A=Multiply(A=@n51, B=@n52.RGB), B=Multiply(A=@n53.RGB, B=@n54))
   B=Add(
    A=ScalarParameter() {defaultValue=1, parameterName=PulseMin}
    B=TextureSample(UVs=@n2) {texture=/Game/Global/Textures/Default/Noise/T_Dudebro_Noise_01}.G
   )
  )
  n56 = MaterialFunctionCall() {materialFunction=/Game/Athena/Environments/Island/CubeGrowth/MF_CubeGrowthTimings}
  n57 = TextureSampleParameter2D() {group=Textures, parameterName=N, samplerType=SAMPLERTYPE_Normal, texture=/Game/Environments/Props/Rocks/Textures/T_Corrupt_Rock_AB_N}
  n58 = MaterialFunctionCall!(SpawnTiming=@n56.SpawnTiming, Normal=@n57.RGB, HitTiming=@n56.SpawnTiming, MaskTexture=OneMinus(None=@n16.R)) {materialFunction=/Game/Athena/Environments/Island/CubeGrowth/MF_CubeGrowth_AdditiveFX_Ice}
  n59 = TextureObjectParameter() {group=Tile, parameterName=N_Tile, samplerType=SAMPLERTYPE_Normal, texture=/Game/Global/Textures/Default/Tileable_Patterns/T_IceScratch1_N}
  n60 = Constant3Vector() {constant=<rgba 0 0 1 0>}
  n61 = VertexNormalWS()
  n62 = LinearInterpolate(
   A=StaticSwitchParameter(
    True=MaterialFunctionCall(TextureObject=@n59, TextureSize=@n12) {materialFunction=&WorldAlignedNormal}.XYZ Texture
    False=TextureSample(UVs=@n13, Tex=@n59) {samplerType=SAMPLERTYPE_Normal, texture=&T_IceScratch1_N}.RGB
   ) {group=Tile, parameterName=Tile WS}
   B=@n60
   Alpha=Abs(None=ComponentMask(None=@n61) {b=true})
  )
  n63 = MaterialFunctionCall(
   BaseNormal=@n57.RGB
   AdditionalNormal=MaterialFunctionCall(
    Normal=LinearInterpolate(A=@n62, B=@n60, Alpha=@n18.Result)
    Flatness=ScalarParameter() {defaultValue=0.85, group=Tile, parameterName=Tile N Flatten}
   ) {materialFunction=&FlattenNormal}.Result
  ) {materialFunction=&BlendAngleCorrectedNormals}
  n64 = Constant() {r=0.2}
  n65 = ScalarParameter() {defaultValue=0.25, group=SSS, parameterName=SSSMin}
  n66 = ScalarParameter() {defaultValue=1, group=SSS, parameterName=SSSMax}
  n67 = SetMaterialAttributes!(
   MaterialAttributes
   Base Color=Multiply(
    A=StaticSwitchParameter(True=@n3.RGB, False=@n4.IceColor) {group=Colors, parameterName=CustomColor}
    B=MaterialFunctionCall(Alpha=@n5, Max=@n6) {materialFunction=&SmoothStep}.Result
   )
   Specular=Constant() {r=0.25}
   Roughness=LinearInterpolate(A=Multiply(A=@n14, B=@n64), Alpha=@n18.Result) {constB=0.025}
   Emissive Color=Add(A=@n58.EmissiveAdd, B=Add(A=@n44, B=@n55))
   Normal=Transform(None=Add(A=@n63.Result, B=@n58.Normal))
   Subsurface Color=Multiply(A=LinearInterpolate(A=@n65, B=@n66, Alpha=@n17.Alpha), B=@n7)
   World Position Offset=@n58.WPO
  ) {attributeSetTypes [
    #1
    #2
    #3
    #4
    #5
    #6
    #7
   ]}
  n68 = MaterialFunctionCall!(UseLandscapeCoordsOnSwitch) {materialFunction=/Game/Athena/Environments/Landscape/MaterialFunctions/Snow/MF_Athena_Snow_01}
  n69 = Constant() {r=0.2}
  n70 = Constant() {r=0.8}
  n71 = ComponentMask(None=MaterialFunctionCall(
    TextureObject=TextureObject() {samplerType=SAMPLERTYPE_LinearColor, texture=/Game/Athena/Environments/Landscape/Textures/T_Athena_Snow_01_M}
    TextureSize=ScalarParameter() {defaultValue=528.9877, parameterName=Specular Flake Tiling}
   ) {materialFunction=&WorldAlignedTexture}.XYZ Texture) {r=true}
  n72 = SetMaterialAttributes!(
   MaterialAttributes=@n68.Material
   Base Color
   Roughness=Clamp(None=LinearInterpolate(A=@n69, B=@n70, Alpha=@n71))
   Specular=LinearInterpolate(Alpha=@n71) {constA=1, constB=0.35}
   Normal
  ) {attributeSetTypes [
    #1
    #3
    #2
    #5
   ]}
  n73 = QualitySwitch(Default=@n72, Low=@n68.Material, High=@n72, Medium=@n68.Material)
  n74 = Constant()
  n75 = ComponentMask(None=Transform(None=@n57.RGB)) {b=true}
  n76 = Multiply(
   A=MaterialFunctionCall(Transition Phase=VertexColor().R, Height Texture=@n16.R) {materialFunction=&HeightLerp}.Alpha
   B=ScalarParameter() {defaultValue=1, group=Colors, parameterName=SnowUpVector}
  )
  n77 = LinearInterpolate(A=Saturate(None=Multiply(A=@n75, B=@n76)), B=Constant() {r=1}, Alpha=@n16.G)
  MaterialFunctionCall!(In=MaterialFunctionCall!(
    Base Material=@n67
    Top Material=SetMaterialAttributes!(MaterialAttributes=@n73, Normal, Emissive Color=@n74, Subsurface Color=@n74) {attributeSetTypes [#5, #4, #6]}
    Alpha=Clamp(None=@n77)
   ) {materialFunction=/Engine/Functions/MaterialLayerFunctions/MatLayerBlend_Standard}.Blended Material) {materialFunction=/Game/Athena/WinterArt/Materials/MF_GameWideShared}
  MaterialFunctionCall!(SpawnTiming, Normal, HitTiming, MaskTexture) {materialFunction=/Game/Athena/Environments/Island/CubeGrowth/MF_CubeGrowth_AdditiveFX}
  MaterialFunctionCall!(In) {materialFunction=/Game/Athena/WinterArt/Materials/MF_Cracks}
  MaterialFunctionCall(BaseNormal=@n63.Result) {materialFunction=&BlendAngleCorrectedNormals}
  ScalarParameter() {defaultValue=3976769536, parameterName=WPOtest}
 ]
}
